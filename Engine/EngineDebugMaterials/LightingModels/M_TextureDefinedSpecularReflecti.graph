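# Engine debug lighting-model material, named for texture-defined specular reflection, using Default Lit shading.
# Diffuse and NormalMap parameters feed a Diffuse+Specular network with light-vector dot product and reflection-style math, remapped to two-channel coordinates.
# A grayscale SpecularReflection texture drives a lerp, alongside a sharpness parameter.
# Many links are absent, suggesting legacy custom-lighting remnants; ConvertFromDiffSpec receives only the diffuse color.
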Material /Engine/EngineDebugMaterials/LightingModels/M_TextureDefinedSpecularReflecti
graph {
 properties {bUsedWithStaticLighting=true, pixelDepthOffsetMode=PDOM_Legacy, usedShadingModels=Default Lit}
 nodes [
  n1 = TextureSampleParameter2D() {parameterName=Diffuse, texture=/Engine/EngineMaterials/DefaultDiffuse}
  Add(A=Multiply(A=Constant() {r=1}), B=Multiply(A=@n1.RGB)) {desc=Diffuse+Specular}
  n2 = Constant() {r=0.5}
  n3 = TextureSampleParameter2D() {parameterName=NormalMap, samplerType=SAMPLERTYPE_Normal, texture=/Engine/EngineMaterials/DefaultNormal}
  n4 = Subtract(A=Add(B=LightVector()), B=@n3.RGB)
  ComponentMask(None=Add(A=Multiply(A=@n2, B=@n4), B=Constant() {r=0.5})) {g=true, r=true}
  Constant()
  DotProduct(A=LightVector(), B=@n3.RGB)
  LinearInterpolate(A=Constant(), B=Constant() {r=-1})
  LinearInterpolate(
   B=LinearInterpolate(A=Constant() {r=1}, B=Constant() {r=1.5})
   Alpha=TextureSampleParameter2D() {parameterName=SpecularReflection, samplerType=SAMPLERTYPE_Grayscale, texture=null}.RGB
  )
  MaterialFunctionCall(DiffuseColor=@n1.RGB) {materialFunction=&ConvertFromDiffSpec}
  ScalarParameter() {parameterName=SpecularReflectionSharpness}
 ]
}
